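; Parent actor for SonicSpear donut drops: falls from the sky with a spawned building prop, damaging what it hits.
; On begin play the server traces ground and water, spawns the colored prop and ground indicator, then starts projectile movement.
; Timers detect height thresholds: the server queues capped building damage; clients play splashes.
; Landing applies final damage, spawns debris, snaps height, aligns horizontal drops, replicates rotation and wobbles the mesh.
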(event EventBeginPlay
  (bind _spawnedprop (Variables|Donut|GetSpawnedProp))
  (bind _outactors (Actor|GetAttachedActors _spawnedprop true true))
  (bind _surfacelinetracelength (Variables|Ground|GetSurfaceLineTraceLength))
  (bind _returnvalue (Math|Vector|MakeVector 0.0 0.0 _surfacelinetracelength))
  (bind _rootcomponent (Variables|Transformation|GetRootComponent))
  (bind (_returnvalue_location _returnvalue_rotation _returnvalue_scale) (Math|Transform|BreakTransform (Transformation|GetWorldTransform _rootcomponent)))
  (bind (_returnvalue_rotation_roll _returnvalue_rotation_pitch _returnvalue_rotation_yaw) (Math|Rotator|BreakRotator _returnvalue_rotation))
  (bind (_returnvalue_x _returnvalue_y _returnvalue_z) (Math|Vector|BreakVector (Transformation|GetActorLocation)))
  (bind _bhasreachedground (Variables|Movement|GetbHasReachedGround))
  (bind _returnvalue_2 (not _bhasreachedground))
  (bind _returnvalue_3 (Transformation|GetActorLocation))
  (bind (_finallocation_x _finallocation_y _finallocation_z) (Math|Vector|BreakVector (CallFunction|GetFinalLocation)))
  (bind (_returnvalue_x_2 _returnvalue_y_2 _returnvalue_z_2) (Math|Vector|BreakVector (Transformation|GetActorLocation)))
  (bind _returnvalue_4 (- _returnvalue_3 _returnvalue))
  (bind _fortprojectilemovement (Variables|Default|GetFortProjectileMovement))
  ;[0 1 0.0211] [BeginPlay]
  (|CallParentFunction)
  (Utilities|FlowControl|Sequence
    (:then_0
      (Variables|/Script/Engine.RotatingMovementComponent|SetRotationRate_Roll (Math|Rotator|MakeRotator 0.0 0.0 (Variables|RotatingMovement|GetRotationRate)) (Variables|Default|GetRotatingMovement)) ; Override RotatingMovement's speed\nwith a hotfixable Config var
    )
    (:then_1
      (Variables|/Script/Engine.ProjectileMovementComponent|SetInitialSpeed (Variables|Movement|GetInitialSpeed) _fortprojectilemovement) ; Override FortProjectileMovement properties with hotfixable Config vars
      (Variables|/Script/Engine.ProjectileMovementComponent|SetMaxSpeed (Variables|Movement|GetMaxSpeed) _fortprojectilemovement)
      (Variables|/Script/Engine.ProjectileMovementComponent|SetProjectileGravityScale (Variables|Movement|GetProjectileGravityScale) _fortprojectilemovement))
    (:then_2
      (EventDispatchers|AssignOnProjectileStop (Variables|Default|GetFortProjectileMovement) (EventDispatchers|CreateEvent :!event Custom|OnMovementStopped))
      (if (and (Variables|Debug|GetbShouldDebug) (Editor|InEditororPIE))
        (bind _returnvalue_5 (Utilities|Time|SetTimerbyEvent (EventDispatchers|CreateEvent :!event Custom|DrawDebug) 1e-05 true true))
        (Variables|Debug|SetDebugTimer _returnvalue_5)))
    (:then_3
      (Networking|SwitchHasAuthority
        (:Authority
          (Utilities|FlowControl|Sequence
            (:then_0
              (Variables|Default|SetSpawnedServerWorldTime (Fortnite|Sync|Time|GetServerWorldTimeSeconds))
              ;;[1 1 1] Determine the location of the ground below
              (bind (_outhit _returnvalue_6) (Collision|LineTraceByChannel _returnvalue_3 _returnvalue_4 "TraceTypeQuery7"))
              (if _returnvalue_6
                (bind (_bblockinghit _binitialoverlap _time _distance _location _impactpoint _normal _impactnormal _physmat _hitactor) (Collision|BreakHitResult _outhit))
                (bind (_location_x _location_y _location_z) (Math|Vector|BreakVector _location))
                (Variables|Ground|SetGroundZ _location_z)
                (Variables|Ground|SetGroundNormal _normal)
                (Development|PrintString (Utilities|String|Append "Found ground [" (Utilities|String|ToString(Vector) (Default|GetGroundLocation)) "] GroundNormal [" (Utilities|String|ToString(Vector) (Variables|Ground|GetGroundNormal)) "] HitActor [" (Utilities|GetDisplayName _hitactor) "]") false)
                (else
                  (Fortnite|FortShippingLog self (Utilities|String|Append "Failed to locate ground between [" (Utilities|String|ToString(Vector) _returnvalue_3) "] and [" (Utilities|String|ToString(Vector) _returnvalue_4) "] Destroying actor.") true)
                  (Actor|DestroyActor))))
            (:then_1
              (if (not (Game|IsActorBeingDestroyed))
                (Utilities|FlowControl|Sequence
                  (:then_0
                    ;;[1 1 1] Determine the location of the water below
                    (bind (_outhit_2 _returnvalue_7) (Collision|LineTraceByChannel _returnvalue_3 (- _returnvalue_3 (Math|Vector|MakeVector 0.0 0.0 (Variables|Ground|GetSurfaceLineTraceLength))) "TraceTypeQuery12"))
                    (if _returnvalue_7
                      (bind _aswater_body (Utilities|Casting|CastToWaterBody (.HitActor (Collision|BreakHitResult _outhit_2))))
                      (Utilities|IsValid (Variables|/Script/Water.WaterBody|GetWaterBodyComponent _aswater_body)
                        (:Is%20Valid
                          (bind (_outwatersurfacelocation _outwatersurfacenormal _outwatervelocity _outwaterdepth) (WaterBody|GetWaterSurfaceInfoatLocation (Variables|/Script/Water.WaterBody|GetWaterBodyComponent _aswater_body) (.ImpactPoint (Collision|BreakHitResult _outhit_2)) true))
                          (bind (_outwatersurfacelocation_x _outwatersurfacelocation_y _outwatersurfacelocation_z) (Math|Vector|BreakVector _outwatersurfacelocation))
                          (Variables|Water|SetWaterZ _outwatersurfacelocation_z)
                          (Variables|Water|SetWaterNormal _outwatersurfacenormal)
                          (Variables|Water|SetWaterDepth _outwaterdepth)
                          (if (>= (Variables|Water|GetWaterZ) (Variables|Ground|GetGroundZ))
                            (Variables|Water|SetbFoundWater true))))))
                  (:then_1
                    ;;[1 1 1] Pick and replicate a random frosting and sprinkle color
                    (if (Variables|Donut|GetSet%20Frosting%20and%20Sprinkle%20Color)
                      (bind (_chosen_colors _valid_color_chosen) (Class|BPFLSonicSpearDonutDropColors|ChooseSonicSpearDropColoring (not (ScalableFloat|GetValueasBool (Variables|Donut|GetUse%20Random%20Coloring%20Ruleset))) "/SonicSpearGameplay/DataTables/SonicSpearDropColorSets.SonicSpearDropColorSets"))
                      (if _valid_color_chosen
                        (Variables|Donut|SetColoring _chosen_colors))))
                  (:then_2
                    ;;[0 0 0 0.5] Spawn donut building prop and attach to this actor for movement and collision logic
                    (bind _returnvalue_8 (Game|SpawnBuildingActorfromClass (Variables|Donut|GetMap%20Prop) (Math|Transform|MakeTransform _returnvalue_location (Math|Rotator|MakeRotator (+ _returnvalue_rotation_roll (Variables|Donut|GetPropRollOffset)) _returnvalue_rotation_pitch _returnvalue_rotation_yaw) _returnvalue_scale) self true (Variables|Donut|GetColoring) 0 0 0))
                    (Utilities|IsValid _returnvalue_8
                      (:Is%20Valid
                        (bind _output_get (Variables|Donut|SetSpawnedProp _returnvalue_8))
                        (Variables|/Script/FortniteGame.BuildingActor|SetbDestroyOnPlayerBuildingPlacement false _output_get) ; Stop destruction from building placement
                        (Transformation|AttachActorToComponent _returnvalue_8 (Variables|Transformation|GetRootComponent) "None" "KeepWorld" "KeepWorld" "KeepWorld" true)
                        (EventDispatchers|BindEventtoOnDied (Variables|Donut|GetSpawnedProp) (EventDispatchers|CreateEvent :!event Custom|HandleMyBuildingOutOfHealth)) ; When the spawned building runs out\nof health, we'll self-destruct
                        (Game|BindEventtoOnDestroyed (Variables|Donut|GetSpawnedProp) (EventDispatchers|CreateEvent :!event Custom|HandleMyBuildingDestroyed)) ; When the spawned building is destroyed,\nwe'll self-destruct
                        (Utilities|Array|Add (Variables|Destruction|GetIgnoreBuildings) _spawnedprop) ; Ignore the SpawnedProp in all of our overlap tests, both for BGAs and non-BGAs, just in case.
                        (Utilities|Array|Add (Variables|PawnHits|GetIgnoreNonBuildings) (Variables|Donut|GetSpawnedProp))
                        (Utilities|Array|AppendArray (Variables|Destruction|GetIgnoreBuildings) _outactors)
                        (Utilities|Array|AppendArray (Variables|PawnHits|GetIgnoreNonBuildings) _outactors)
                        (GameplayCue|ExecuteGameplayCueOnActor(Burst) self (Variables|Cues|GetGC_Spawn) (Ability|GameplayCue|MakeGameplayCueParameters 0.0)))
                      (:Is%20Not%20Valid
                        (Fortnite|FortShippingLog self (Utilities|String|Append "SpawnBuildingActor failed for class [" (Utilities|GetClassDisplayName (Variables|Donut|GetMap%20Prop)) "] Loc [" (Utilities|String|ToString(Vector) _returnvalue_location) "] Destroying this actor.") true)
                        (Actor|DestroyActor))))
                  (:then_3
                    (if (not (Game|IsActorBeingDestroyed))
                      (Utilities|FlowControl|Sequence
                        (:then_0
                          ;;[1 1 1] Spawn the GroundTargetActor
                          (if (ScalableFloat|GetValueasBool (Variables|Default|GetGround%20Indicator%20Enabled))
                            (bind _returnvalue_9 (Game|SpawnActorfromClass "/SonicSpearGameplay/Gameplay/Drops/B_SonicSpear_GroundTargetActor.B_SonicSpear_GroundTargetActor_C" (Math|Transform|MakeTransform (Default|GetGroundLocation)) "AlwaysSpawn" "OverrideRootScale" 0 self (CallFunction|GetSpawnedPropMesh) 0 "0.000000,0.000000,0.000000" false 0)) ; spawn ground actor feedback and pass newly spawned drop object
                            (Utilities|IsValid _returnvalue_9
                              (:Is%20Valid
                                (Variables|Default|SetGroundTargetActor _returnvalue_9)
                                (Utilities|Array|Add (Variables|Destruction|GetIgnoreBuildings) _returnvalue_9) ; Ignore the SpawnedProp in all of our overlap tests, both for BGAs and non-BGAs, just in case.
                                (Utilities|Array|Add (Variables|PawnHits|GetIgnoreNonBuildings) _returnvalue_9))
                              (:Is%20Not%20Valid
                                (Fortnite|FortShippingLog self (Utilities|String|Append "Failed to spawn  GroundTargetActor at Loc [" (Utilities|String|ToString(Vector) (Default|GetGroundLocation)) "]") true)))))
                        (:then_1
                          ;;[1 1 1] Set up components on server
                          (Components|DestroyComponent (Variables|Default|GetAudio)) ; Destroy cosmetic components on server, so the server\ndoesn't pay the cost of updating them during the fall
                          (Components|Sphere|SetSphereRadius (Variables|Default|GetNonBuildingCollision) (Variables|PawnHits|GetPawnOverlapRadius) false) ; Change collision radius using\nhotfixable value
                          (Collision|SetCollisionEnabled (Variables|Default|GetNonBuildingCollision) "QueryAndPhysics") ; Enable NonBuildingCollision on server only.\nIts default setting is NoCollision.\nThis component will be destroyed on client.
                          (Projectile|SetVelocity (Variables|Default|GetFortProjectileMovement) (Math|Vector|MakeVector 0.0 0.0 (* (Variables|Movement|GetInitialSpeed) -1.0))) ; Start with a downward velocity
                          (Variables|/Script/Engine.ProjectileMovementComponent|SetbSweepCollision (Variables|Default|GetbSweepCollisionOnServer) (Variables|Default|GetFortProjectileMovement)) ; FortProjectileMovement's bSweepCollision\ndefaults to false.\nUse a hotfix to possibly enable it on server.
                          (if (Variables|Default|GetbShouldRotate)
                            (Components|Activation|Activate (Variables|Default|GetRotatingMovement) false) ; Begin server rotation.\nAutoActivate is false.
                          ))
                        (:then_2
                          ;;[1 1 1] On server, set up a recurring timer to check for buildings to damage
                          (Utilities|Array|AddUnique (Variables|Destruction|GetIgnoreBuildings) self)
                          (bind _returnvalue_10 (Utilities|Time|SetTimerbyEvent (EventDispatchers|CreateEvent :!event Custom|ServerUpdate) (Variables|Update|GetServerUpdateTime) true true))
                          (Variables|Update|SetUpdateTimer _returnvalue_10))))))))))
        (:Remote
          ;;[1 1 1] Client setup
          (Components|DestroyComponent (Variables|Default|GetNonBuildingCollision)) ; NonBuildingCollision is not used on client.  Destroy it to not\nhave to pay the cost of moving it every frame during the fall.
          (Variables|Movement|SetbHasReachedWater (<= _returnvalue_z (Variables|Water|GetWaterZ))) ; If we're already below water, do not monitor for it in ClientUpdate.\nDo not call ReachedWater, which plays fx and audio.
          (Variables|Movement|SetbHasReachedGround (<= _returnvalue_z (Default|GetReachedGroundZ))) ; Likewise, for ReachedGroundZ\nand the ReachedGround event.
          (if (or (CallFunction|HasStoppedMovement) (<= _returnvalue_z_2 _finallocation_z))
            (Utilities|IsValid (CallFunction|GetSpawnedPropMesh)
              (:Is%20Valid
                (CallFunction|SetMobilitytoStatic)
                (if (Variables|Default|GetbShouldAutoCorrectClientAttachment)
                  (Utilities|Time|SetTimerforNextTickbyFunctionName 0 "FixClientAttachment") ; Wait one frame before attempting to fix client attachment.\nIt will hopefully be dealt with in AActor::OnRep_AttachmentReplication() before then.
                ))) ; If the SpawnedProp isn't valid here,\nwe'll later get OnRep_SpawnedProp\nand call the following logic from there
            (else
              (bind _returnvalue_11 (Utilities|Time|SetTimerbyEvent (EventDispatchers|CreateEvent :!event Custom|ClientUpdate) (Variables|Update|GetClientUpdateTime) true true)) ; Recurring timer to check for crossing each Z threshold
              (Variables|Update|SetUpdateTimer _returnvalue_11)
              (Audio|Parameter|SetBooleanParameter (Variables|Default|GetAudio) "AudioIncoming" _returnvalue_2 :!class /Script/Engine.AudioComponent) ; Play the fall sound if we haven't\nreached the ground
              (if _returnvalue_2
                (Components|Activation|Activate (Variables|Default|GetNS_SonicSpear_Fall) false) ; Begin the fall fx if we haven't\nreached the ground
                (Components|Activation|Activate (Variables|Default|GetAudio) false) ; Begin the audio if we haven't\nreached the ground
                (Variables|/Script/Engine.ProjectileMovementComponent|SetbSweepCollision (Variables|Default|GetbSweepCollisionOnClient) (Variables|Default|GetFortProjectileMovement)) ; FortProjectileMovement's bSweepCollision\ndefaults to false.\nUse a hotfix to possibly enable it on client.
                (if (Variables|Default|GetbShouldRotate)
                  (Components|Activation|Activate (Variables|Default|GetRotatingMovement) false) ; Begin rotating if we haven't\nreached the ground
                )))))))))

(event Custom|ClientUpdate
  (bind (_finallocation_x _finallocation_y _finallocation_z) (Math|Vector|BreakVector (CallFunction|GetFinalLocation)))
  (bind (_returnvalue_x _returnvalue_y _returnvalue_z) (Math|Vector|BreakVector (Transformation|GetActorLocation)))
  ;[1 0.5 0 0.5] [ClientUpdate]
  (Utilities|FlowControl|Sequence
    (:then_0
      ;;[1 1 1] Detect when we cross the WaterZ on the client
      (CallFunction|DetectReachedWater))
    (:then_1
      ;;[1 1 1] Detect when the bottom of the donut crosses the ReachedGroundZ on the client
      (if (and (not (Variables|Movement|GetbHasReachedGround)) (<= _returnvalue_z (Default|GetReachedGroundZ)))
        (Utilities|FlowControl|DoOnce false
          (:Completed
            (CallFunction|ReachedGround)))))
    (:then_2
      ;;[1 1 1] Detect when we cross the FinalLocation.Z on the client, in case our simulation gets there before the server stops and replicates the stop
      (if (and (not (CallFunction|HasStoppedMovement)) (<= _returnvalue_z _finallocation_z))
        (Game|Components|ProjectileMovement|StopSimulating (Variables|Default|GetFortProjectileMovement) (Collision|MakeHitResult true false 0.0 0.0 (CallFunction|GetFinalLocation)))))))

(event Custom|ServerUpdate
  (bind (_returnvalue_x _returnvalue_y _returnvalue_z) (Math|Vector|BreakVector (Transformation|GetActorLocation)))
  (bind (_finallocation_x _finallocation_y _finallocation_z) (Math|Vector|BreakVector (CallFunction|GetFinalLocation)))
  (bind (_returnvalue_x_2 _returnvalue_y_2 _returnvalue_z_2) (Math|Vector|BreakVector (Transformation|GetActorLocation)))
  (bind _damagebuildings (Variables|Destruction|GetDamageBuildings))
  (bind _output (Utilities|Array|Get(acopy) _damagebuildings))
  ;[1 1 0 0.5] [ServerUpdate] Repeating timer set up in BeginPlay
  (Utilities|FlowControl|Sequence
    (:then_0
      (if (not (Variables|Movement|GetbHasReachedGround))
        (Utilities|FlowControl|Sequence
          (:then_0
            ;;[0.09493 1 0] Collect any newly overlapped actors and queue them to be damaged.
            (bind _outactors (Collision|SphereOverlapActors (+ (Transformation|GetActorLocation) (Variables|Destruction|GetOverlapOffset)) (Variables|Destruction|GetOverlapBuildingRadius) (Variables|Destruction|GetDamageableBuildingObjectTypes) "/Script/FortniteGame.BuildingActor" (Variables|Destruction|GetIgnoreBuildings)))
            (Utilities|Array|AppendArray (Variables|Destruction|GetDamageBuildings) _outactors)
            (Utilities|Array|AppendArray (Variables|Destruction|GetIgnoreBuildings) _outactors))
          (:then_1
            ;;[0.05306 0.3743 1] Attempt to deal damage to the maximum number of queued buildings allowed per trace.
            (Utilities|FlowControl|ForLoopwithBreak 0 (- (Variables|Destruction|GetMaxBuildingsToDamagePerTrace) 1)
              (:LoopBody
                (if (Utilities|Array|IsValidIndex (Variables|Destruction|GetDamageBuildings))
                  _
                  (else
                    _))
                (Utilities|FlowControl|Sequence
                  (:then_0
                    ;;;[0.1041 1 0.6523] Attempt to deal damage to the building that was hit.
                    (Utilities|IsValid _output
                      (:Is%20Valid
                        (bind _returnvalue (Utilities|MakeMyFortContext))
                        (GameplayEffects|ApplyGameplayEffectToTarget (Ability|GetAbilitySystemComponent self) (Variables|GEs|GetGE_EnvironmentalDamage) (Ability|GetAbilitySystemComponent _output) 0.0 _returnvalue))))
                  (:then_1
                    ;;;[1 0.5861 0.08121] Remove the building from the queue of buildings to damage.
                    (Utilities|Array|RemoveIndex (Variables|Destruction|GetDamageBuildings))))))))))
    (:then_1
      ;;[1 1 1] Detect when the bottom of the donut crosses the ReachedGroundZ on the server
      (if (and (not (Variables|Movement|GetbHasReachedGround)) (<= _returnvalue_z (Default|GetReachedGroundZ)))
        (CallFunction|ReachedGround)))
    (:then_2
      ;;[1 1 1 1 17] Detect when we cross the FinalLocation.Z on the server, and StopSimulating.
      ;; This will trigger our OnMovementStopped event on both server and client, via StopSimulatingRepData.
      ;; There, we snap our Z back up to the FinalLocation.Z.
      ;; Note that we're also overriding GetMaxTravelDistanceSquared to limit the distance traveled,
      ;; though allowing that to go slightly farther, so that the <= check below will be able to succeed.
      (if (and (not (CallFunction|HasStoppedMovement)) (<= _returnvalue_z_2 _finallocation_z))
        (Game|Components|ProjectileMovement|StopSimulating (Variables|Default|GetFortProjectileMovement) (Collision|MakeHitResult true false 0.0 0.0 (CallFunction|GetFinalLocation)))))))

(event Custom|DetectReachedWater
  (bind (_returnvalue_x _returnvalue_y _returnvalue_z) (Math|Vector|BreakVector (Transformation|GetActorLocation)))
  ;[0 0 1 0.5] [DetectReachedWater] Client only
  (if (Variables|Movement|GetbHasReachedWater)
    _
    (elif (and (Variables|Water|GetbFoundWater) (<= _returnvalue_z (Default|GetReachedWaterZ)))
      (Variables|Movement|SetbHasReachedWater true)
      (Networking|CanExecuteCosmeticEvents
        (:True
          (bind _returnvalue (Niagara|SpawnSystematLocation "/SonicSpearGameplay/Assets/VFX/NS_SonicSpear_WaterSplash.NS_SonicSpear_WaterSplash" (Default|GetWaterLocation)))
          (Niagara|SetNiagaraVariable(Vector3) _returnvalue "WaterPlaneLocation" (Default|GetWaterLocation))
          (Niagara|SetNiagaraVariable(Vector3) _returnvalue "WaterPlaneNormal" (Variables|Water|GetWaterNormal))
          (Niagara|SetNiagaraVariable(Float) _returnvalue "WaterDepth" (Variables|Water|GetWaterDepth))
          (Audio|PlaySoundatLocation "/Game/Athena/Sounds/Weapons/Explosions/Expl_Debris_Water_Cue.Expl_Debris_Water_Cue" (Default|GetWaterLocation))
          (Development|PrintString "I hit water!" true true "(R=0.000000,G=0.660000,B=1.000000,A=1.000000)" 5.0))))))

(event OnComponentBeginOverlap(NonBuildingCollision)
  (params (OverlappedComponent :type Primitive%20Component%20Object%20Reference) (OtherActor :type Actor%20Object%20Reference) (OtherComp :type Primitive%20Component%20Object%20Reference) (OtherBodyIndex :type Integer) (bFromSweep :type Boolean) (SweepResult :type Hit%20Result%20Structure%20%28by%20ref%29))
  ;[0.03857 0.1424 0.15 0.5] [NonBuildingCollision - OnComponentBeginOverlap] On overlap with non-buildings (pawns, vehicles, physics objects) apply impulse and/or damage
  (Networking|SwitchHasAuthority
    (:Authority
      (if (not (Variables|Movement|GetbHasReachedGround))
        (CallFunction|ProcessNonBuildingHit OtherActor (Variables|GEs|GetGE_VehicleDamage) (Variables|GEs|GetGE_PlayerDamage) (Variables|GEs|GetGE_PhysicsObjectDamage))))))

(event Custom|ReachedGround
  (bind (_returnvalue_x _returnvalue_y _returnvalue_z) (Math|Vector|BreakVector (Transformation|GetWorldScale (Variables|/Script/FortniteGame.BuildingSMActor|GetStaticMeshComponent (Variables|Donut|GetSpawnedProp)))))
  ;[0.05067 0.15 0.06199 0.5] [ReachedGround]
  (CallFunction|DetectReachedWater) ; In case we reached the ground, but hadn't yet\nprocessed crossing through the water
  (if (not (Variables|Movement|GetbHasReachedGround))
    (Variables|Movement|SetbHasReachedGround true)
    (Networking|SwitchHasAuthority
      (:Authority
        (Collision|SetCollisionEnabled (Variables|Default|GetNonBuildingCollision) "NoCollision") ; No further need for the NonBuildingCollision.\nDisable collision to prevent further overlaps.
        (bind _outactors (Collision|SphereOverlapActors (CallFunction|GetFinalLocation) (Variables|PawnHits|GetFinalPawnOverlapRadius) (Variables|PawnHits|GetFinalHitObjectType) 0 (Variables|PawnHits|GetIgnoreNonBuildings))) ; Manually perform a final overlap test for pawns and vehicles,\nrather than relying on NonBuildingCollision
        (for _array_element _outactors
          (CallFunction|ProcessNonBuildingHit _array_element (Variables|GEs|GetGE_FinalVehicleDamage) (Variables|GEs|GetGE_FinalPlayerDamage) (Variables|GEs|GetGE_FinalPhysicsObjectDamage)) ; Apply an alternate damage GE to pawns\nand vehicles for the final hit
        )
        (CallFunction|DamageRemainingQueuedBuildings) ; Damage buildings that we had gathered in previous\noverlap tests, but exceeded the per-frame limit
        (bind _outactors_2 (Collision|SphereOverlapActors (CallFunction|GetFinalLocation) (Variables|Destruction|GetFinalEnvOverlapRadius) (Variables|Destruction|GetDamageableBuildingObjectTypes) "/Script/FortniteGame.BuildingActor" (Variables|Destruction|GetIgnoreBuildings))) ; Perform a final overlap versus buildings
        (for _array_element_2 _outactors_2
          (bind _returnvalue (Utilities|MakeMyFortContext))
          (GameplayEffects|ApplyGameplayEffectToTarget (Ability|GetAbilitySystemComponent self) (Variables|GEs|GetGE_FinalEnvironmentalDamage) (Ability|GetAbilitySystemComponent _array_element_2) 0.0 _returnvalue) ; Apply an alternate damage GE\nto buildings for the final hit
        )
        ;;[1 1 1] Spawn debris actor
        (bind _returnvalue_2 (GameplayTags|HasMatchingGameplayTag(Message) self (Variables|Debris|GetPreventDebrisForTags)))
        (if (not (or _returnvalue_2 (Variables|Water|GetbFoundWater)))
          (Game|SpawnActorfromClass (Variables|Debris|GetDebrisClass) (Math|Transform|MakeTransform (CallFunction|GetFinalLocation) (Transformation|GetActorRotation)) "AlwaysSpawn" "MultiplyWithRoot" 0 (Variables|Donut|GetColoring) _returnvalue_z (Variables|Donut|GetSpawnedProp) 0)))
      (:Remote
        (CallFunction|StopIntroScaleTimeline)))))

(event Custom|StopIntroScaleTimeline
  (bind _introscaletimeline (Variables|BGA_SonicSpear_MainDropParent|GetIntroScaleTimeline))
  ;[1 0.5 0 0.5] [StopIntroScaleTimeline]
  (Networking|CanExecuteCosmeticEvents
    (:True
      (if (Components|Timeline|IsPlaying _introscaletimeline)
        (Components|Timeline|Stop _introscaletimeline)
        (CallFunction|RestoreOriginalScale)))))

(event Custom|DoIntroScale
  (bind _static_mesh_component (CallFunction|GetSpawnedPropMesh))
  ;[0 0.1461 0.15 0.5] [DoIntroScale] If we're net relevant soon enough after the time of spawning, scale up the donut.  Don't do this after we've fallen too far, or already landed.
  (Utilities|FlowControl|DoOnce false
    (:Completed
      (Networking|CanExecuteCosmeticEvents
        (:True
          (if (Variables|Default|GetbShouldPlayIntroScale)
            (if (Variables|Default|GetbHasPropStartingScaleBeenCached)
              (if (Variables|Movement|GetbHasReachedGround)
                _
                (elif (Components|Timeline|IsPlaying (Variables|BGA_SonicSpear_MainDropParent|GetWobbleTimeline))
                  _
                  (elif (< (- (Fortnite|Sync|Time|GetServerWorldTimeSeconds) (Variables|Default|GetSpawnedServerWorldTime)) (Variables|Default|GetMaxTimeToInitiateIntroScale))
                    (if (> (Variables|Default|GetIntroScaleDuration) 0)
                      (Components|Timeline|SetPlayRate (Variables|BGA_SonicSpear_MainDropParent|GetIntroScaleTimeline) (/ (Components|Timeline|GetTimelineLength (Variables|BGA_SonicSpear_MainDropParent|GetIntroScaleTimeline)) (Variables|Default|GetIntroScaleDuration))) ; Hotfix the timeline duration
                      (bind (_direction _scale) (Timeline|IntroScaleTimeline PlayFromStart)
                        (:Update
                          (Utilities|IsValid _static_mesh_component
                            (:Is%20Valid
                              (Transformation|SetWorldScale3D _static_mesh_component (* (Variables|Wobble|GetPropStartingScale) _scale)))))
                        (:Finished
                          (CallFunction|RestoreOriginalScale))))))))))))))

(event Custom|RestoreOriginalScale
  (bind _static_mesh_component (CallFunction|GetSpawnedPropMesh))
  ;[0.15 0.01802 0.02361 0.5] [RestoreOriginalScale]
  (Networking|CanExecuteCosmeticEvents
    (:True
      (if (Variables|Default|GetbHasPropStartingScaleBeenCached)
        (Utilities|IsValid _static_mesh_component
          (:Is%20Valid
            (Transformation|SetWorldScale3D _static_mesh_component (Variables|Wobble|GetPropStartingScale))))))))

(event Custom|StopWobbleTimeline
  (bind _wobbletimeline (Variables|BGA_SonicSpear_MainDropParent|GetWobbleTimeline))
  ;[0 0.07528 0.15 0.5] [StopWobbleTimeline]
  (Networking|CanExecuteCosmeticEvents
    (:True
      (if (Components|Timeline|IsPlaying _wobbletimeline)
        (Components|Timeline|Stop _wobbletimeline)
        (CallFunction|RestoreOriginalScale)))))

(event Custom|DoWobble
  (bind _static_mesh_component (CallFunction|GetSpawnedPropMesh))
  ;[0.15 0.1296 0.02393 0.5] [DoWobble] Wobble Mesh on finish.  Note that this will only trigger if the donut is net relevant at the moment the donut lands.
  ; The calling function, OnMovementStopped, will not occur on client if the donut landed prior to spawning on client.
  (Utilities|FlowControl|DoOnce false
    (:Completed
      (Networking|CanExecuteCosmeticEvents
        (:True
          (if (Variables|Default|GetbCanWobbleOnLand)
            (if (Variables|Default|GetbHasPropStartingScaleBeenCached)
              (CallFunction|StopIntroScaleTimeline)
              (if (> (Variables|Default|GetWobbleDuration) 0)
                (Components|Timeline|SetPlayRate (Variables|BGA_SonicSpear_MainDropParent|GetWobbleTimeline) (/ (Components|Timeline|GetTimelineLength (Variables|BGA_SonicSpear_MainDropParent|GetWobbleTimeline)) (Variables|Default|GetWobbleDuration))) ; Hotfix the timeline duration
                (bind (_direction _squashz) (Timeline|WobbleTimeline PlayFromStart)
                  (:Update
                    (Utilities|IsValid _static_mesh_component
                      (:Is%20Valid
                        (Transformation|SetWorldScale3D _static_mesh_component (* (Variables|Wobble|GetPropStartingScale) (Math|Vector|MakeVector (+ 1.0 (* _squashz (Variables|Wobble|GetWobbleIntensityScalar))) (- 1.0 (* _squashz (Variables|Wobble|GetWobbleIntensityScalar))) 1.0))))))
                  (:Finished
                    (CallFunction|RestoreOriginalScale)))))))))))

(event Custom|OnMovementStopped
  (params (ImpactResult :type Hit%20Result%20Structure%20%28by%20ref%29))
  (bind (_bblockinghit _binitialoverlap _time _distance _location) (Collision|BreakHitResult ImpactResult))
  (bind _ns_sonicspear_fall (Variables|Default|GetNS_SonicSpear_Fall))
  ;[0.15 0.02844 0.1288 0.5 16] [OnMovementStopped] FortProjectileComponent's ReplicateStopSimulating = true.
  ; When StopSimulating is called on server, both the server and client will receive OnProjectileStop.
  ; Additionally, we have the client call OnProjectileStop on its own, if it reaches FinalLocation.Z before it has been told by the server.
  (Utilities|FlowControl|DoOnce false
    (:Completed
      (CallFunction|ReachedGround) ; Force a call to ReachedGround, in case we stopped\nbefore detecting the ground in Server/ClientUpdate.
      (Utilities|FlowControl|Sequence
        (:then_0
          (bind _nonbuildingcollision (Variables|Default|GetNonBuildingCollision)
            (:then
              (Components|DestroyComponent _nonbuildingcollision))
            (:else
              _)) ; Destroy collision before we teleport,\nso we don't trigger further overlaps,\nand so we don't pay for movement.
          (Transformation|SetActorLocation _location false true) ; Force our Z to the desired final Z
        )
        (:then_1
          ;;[1 1 1] Align horizontal donuts to the GroundNormal.
          ;; If the client already received OnRep_FinalRotation and synced with the server, then don't do this for the client.
          (if (== (Variables|Donut|GetDrop%20Type) "(TagName=\"SonicSpear.DropType.Horizontal\")")
            (if (or (Networking|IsServer) (not (Variables|Default|GetbHasClientAppliedFinalRotation)))
              (Transformation|SetActorRotation self (Math|Rotator|MakeRotfromZ (Variables|Ground|GetGroundNormal)) true))))
        (:then_2
          (Networking|SwitchHasAuthority
            (:Authority
              (Utilities|FlowControl|Sequence
                (:then_0
                  (Variables|Default|SetFinalRotation (Transformation|GetActorRotation)) ; Replicate the final rotation
                )
                (:then_1)))))
        (:then_3
          (Utilities|Time|ClearandInvalidateTimerbyHandle (Variables|Update|GetUpdateTimer))
          (Components|Activation|Deactivate (Variables|Default|GetRotatingMovement))
          (CallFunction|SetMobilitytoStatic)
          (bind _spawnedprop (Variables|Donut|GetSpawnedProp)
            (:then
              (Fortnite|SetCanbeDamaged _spawnedprop true) ; Allow damage to our BGA\non both server and client
            )
            (:else
              _))
          (Networking|ForceNetUpdate))
        (:then_4
          (Networking|CanExecuteCosmeticEvents
            (:True
              (Audio|Components|Audio|FadeOut (Variables|Default|GetAudio) 0.1 0.0 "Linear" :!class /Script/Engine.AudioComponent)
              (GameplayCue|ExecuteGameplayCue%28Burst%29Local self (Variables|Cues|GetGC_Destroy) (Ability|GameplayCue|MakeGameplayCueParameters :NormalizedMagnitude 0.0 :RawMagnitude 0.0 :Location (CallFunction|GetFinalLocation) :Normal (Variables|Ground|GetGroundNormal) :Instigator self :EffectCauser self :SourceObject self :GameplayEffectLevel 1 :AbilityLevel 1 :bReplicateLocationWhenUsingMinimalRepProxy false))
              (Fortnite|ShowSoundIndicatorLocalforTeam :Location (Transformation|GetActorLocation) :MaxAudibleDistance 4e+03 :IndicatorType "Gunshot" :Team (Team|GetActorTeam self) :Affiliations (Utilities|Array|MakeArray "Hostile" "Neutral") :Tint "(R=0.000000,G=0.000000,B=0.000000,A=0.000000)" :AttenuationMaxRangeMultiplier 1.0 :AttenuationIntensityMultiplier 1.0)
              (CallFunction|DoWobble)
              (Niagara|SetNiagaraVariable(Bool) _ns_sonicspear_fall "StopSpawning" true)
              (Utilities|FlowControl|Delay (Variables|Default|GetDelayBeforeDeactivateNSSonicSpearFall))
              (Components|Activation|Deactivate _ns_sonicspear_fall))))
        (:then_5
          (Default|CallOnLanded))))))

(event Custom|DamageRemainingQueuedBuildings
  ;[1 0.4871 0.0636] [DamageRemainingQueuedBuildings] Damage any buildings that are still queued for damage.
  ; Called when the final hit occurs, before dealing final damage to buildings.
  (for _array_element (Variables|Destruction|GetDamageBuildings)
    (Utilities|IsValid _array_element
      (:Is%20Valid
        (bind _returnvalue (Utilities|MakeMyFortContext))
        (GameplayEffects|ApplyGameplayEffectToTarget (Ability|GetAbilitySystemComponent self) (Variables|GEs|GetGE_EnvironmentalDamage) (Ability|GetAbilitySystemComponent _array_element) 0.0 _returnvalue))))
  (Development|PrintString (Utilities|String|ToString(Text) (Utilities|Text|FormatText "NSLOCTEXT(\"[88CB3AC7A0F8EECA0306420CF0502298]\", \"9BD84C8F41A776A24239158924EE7D74\", \"QueuedBuildings: {X}  IgnoredBuildings: {Y}\")" (Utilities|Array|Length (Variables|Destruction|GetDamageBuildings)) (Utilities|Array|Length (Variables|Destruction|GetIgnoreBuildings)))))
  (Utilities|Array|Clear (Variables|Destruction|GetDamageBuildings)))

(event Custom|DoSilentDie
  ;[0.15 0.887 0.15 0.5] [DoSilentDie]
  (Gameplay|SilentDie false) ; Self-destruct.  Internally, this is not instant, but very short.
)

(event Custom|HandleMyBuildingOutOfHealth
  (params (DamagedActor :type Actor%20Object%20Reference) (Damage :type Float%20%28single-precision%29) (InstigatedBy :type Controller%20Object%20Reference) (DamageCauser :type Actor%20Object%20Reference) (HitLocation :type Vector) (FHitComponent :type Primitive%20Component%20Object%20Reference) (BoneName :type Name) (Momentum :type Vector))
  ;[0.15 0.02244 0 0.5] [HandleMyBuildingOutOfHealth]
  (if (== DamagedActor (Variables|Donut|GetSpawnedProp))
    (CallFunction|SelfDestruct false)))

(event Custom|HandleMyBuildingDestroyed
  (params (DestroyedActor :type Actor%20Object%20Reference))
  ;[0 0.08964 0.15 0.5] [HandleMyBuildingDestroyed]
  (if (== DestroyedActor (Variables|Donut|GetSpawnedProp))
    (Variables|Default|SetbShouldStopEverything true)))

(event Custom|SelfDestruct
  (params (bDetach :type Boolean) (bWaitForVisuals :type Boolean))
  ;[0.06413 0 0.15 0.5] [SelfDestruct]
  (Utilities|FlowControl|DoOnce false
    (:Completed
      (Networking|SwitchHasAuthority
        (:Authority
          (Utilities|FlowControl|Sequence
            (:then_0
              (Variables|Default|SetbShouldStopEverything true))
            (:then_1
              (if bDetach
                (Transformation|DetachFromActor (Variables|Donut|GetSpawnedProp) "KeepWorld" "KeepWorld" "KeepWorld")))
            (:then_2
              (if bWaitForVisuals
                (bind _returnvalue (Utilities|Time|SetTimerbyEvent (EventDispatchers|CreateEvent :!event Custom|DoSilentDie) (Math|Float|Max(Float) (Variables|Default|GetDelayBeforeDeactivateNSSonicSpearFall) (Variables|Default|GetIntroScaleDuration) (Variables|Default|GetWobbleDuration) 0.1))) ; Self-destruct timer
                (Variables|Default|SetSelfDestructTimer _returnvalue)
                (else
                  (CallFunction|DoSilentDie))))))))))

(event Custom|StopEverything
  ;[0.1211 0.15 0 0.5] [StopEverything]
  (Utilities|FlowControl|DoOnce false
    (:Completed
      (Utilities|Time|ClearandInvalidateTimerbyHandle (Variables|Update|GetUpdateTimer))
      (EventDispatchers|UnbindEventfromOnProjectileStop (Variables|Default|GetFortProjectileMovement) (EventDispatchers|CreateEvent :!event Custom|OnMovementStopped))
      (bind _spawnedprop (Variables|Donut|GetSpawnedProp)
        (:then
          (EventDispatchers|UnbindEventfromOnOutOfHealthNotDestroyed _spawnedprop (EventDispatchers|CreateEvent :!event Custom|HandleMyBuildingOutOfHealth))
          (Game|UnbindEventfromOnDestroyed _spawnedprop (EventDispatchers|CreateEvent :!event Custom|HandleMyBuildingDestroyed)))
        (:else
          _))
      (Collision|SetCollisionEnabled (Variables|Default|GetNonBuildingCollision) "NoCollision")
      (Components|Activation|Deactivate (Variables|Default|GetRotatingMovement))
      (bind _ns_sonicspear_fall (Variables|Default|GetNS_SonicSpear_Fall)
        (:then
          (Components|Activation|Deactivate _ns_sonicspear_fall))
        (:else
          _))
      (bind _audio (Variables|Default|GetAudio)
        (:then
          (Components|Activation|Deactivate _audio))
        (:else
          _))
      (CallFunction|StopIntroScaleTimeline)
      (CallFunction|StopWobbleTimeline))))

(event EventEndPlay
  (params (EndPlayReason :type EEndPlayReason%20Enum :default "Destroyed"))
  ;[1 0 0 0.5] [EndPlay]
  (|CallParentFunction EndPlayReason)
  (CallFunction|StopEverything)
  (Utilities|Time|ClearandInvalidateTimerbyHandle (Variables|Debug|GetDebugTimer))
  (Utilities|Time|ClearandInvalidateTimerbyHandle (Variables|Default|GetSelfDestructTimer)))

(event Custom|DrawDebug
  (bind _returnvalue (Networking|IsServer))
  (bind (_returnvalue_x _returnvalue_y) (Math|Vector|BreakVector (Transformation|GetActorLocation)))
  (bind _returnvalue_2 20)
  (bind _returnvalue_3 5.0)
  (bind _pawnoverlapradius (Variables|PawnHits|GetPawnOverlapRadius))
  (bind _returnvalue_4 (* _pawnoverlapradius 0.8))
  (bind _returnvalue_5 (Math|Float|SelectFloat _pawnoverlapradius _returnvalue_4 _returnvalue))
  ;[0.15 0.1337 0.001048 0.5] [DrawDebug]
  (if (Variables|Debug|GetbShouldDebug)
    (Rendering|Debug|DrawDebugCircle (Transformation|GetActorLocation) _returnvalue_5 _returnvalue_2 "(R=1.000000,G=1.000000,B=0.000000,A=1.000000)" 0.0 _returnvalue_3 "0.000000,1.000000,0.000000" "1.000000,0.000000,0.000000")
    (Rendering|Debug|DrawDebugCircle (Math|Vector|MakeVector _returnvalue_x _returnvalue_y (Default|GetBottomZ)) _returnvalue_5 _returnvalue_2 "(R=1.000000,G=0.500000,B=0.000000,A=1.000000)" 0.0 _returnvalue_3 "1.000000,0.000000,0.000000" "0.000000,1.000000,0.000000")
    (Rendering|Debug|DrawDebugCircle (CallFunction|GetFinalLocation) _returnvalue_5 _returnvalue_2 "(R=1.000000,G=0.000000,B=0.000000,A=1.000000)" 0.0 _returnvalue_3 "1.000000,0.000000,0.000000" "0.000000,1.000000,0.000000")
    (Rendering|Debug|DrawDebugCircle (Default|GetGroundLocation) _returnvalue_5 _returnvalue_2 "(R=0.000000,G=1.000000,B=0.000000,A=1.000000)" 0.0 _returnvalue_3 "0.000000,1.000000,0.000000" "1.000000,0.000000,0.000000")
    (if (Variables|Water|GetbFoundWater)
      (Rendering|Debug|DrawDebugCircle (Default|GetWaterLocation) _returnvalue_5 _returnvalue_2 "(R=0.000000,G=0.000000,B=1.000000,A=1.000000)" 0.0 _returnvalue_3 "1.000000,0.000000,0.000000" "0.000000,1.000000,0.000000"))))

(fn FixClientAttachment
  (bind _static_mesh_component (CallFunction|GetSpawnedPropMesh))
  (bind _spawnedprop (Variables|Donut|GetSpawnedProp))
  (bind _returnvalue (Transformation|GetActorLocation _spawnedprop))
  ;[1 1 1] On client, we experienced an error in USceneComponent::AttachToComponent() via AActor::OnRep_AttachmentReplication().
  ; Basically, this actor "is not static... cannot attach" the donut-actor "which is static to it.  Aborting."
  ; Our Mobility was Movable, but the donut was Static.
  ; The problem was that the BP logic hadn't been set up to set our Mobility to Static on the server.
  ; After fixing that, now when the donut lands on the server, it sets the Mobility of both actors to Static.
  ; This replicates to the client, and we no longer receive the above error.  It used to be easy to repro in PIE.
  ; Additionally, we were sometimes receiving a warning of movement replication failing.  That no longer occurs, either.
  ; I had made this function in an earlier attempt to fix the client attachment after the fact, before I made the proper Mobility fix.
  ; Still, I left the function here just in case, gated by a hotfix bool, defaulted to false.
  (Networking|CanExecuteCosmeticEvents
    (:True
      (if (Variables|Default|GetbShouldAutoCorrectClientAttachment)
        (Utilities|IsValid _static_mesh_component
          (:Is%20Valid
            (if (!= (Transformation|GetAttachParent _static_mesh_component) (Variables|Default|GetRootScene))
              (LocalVariables|SetSpawnedPropPrevLoc _returnvalue)
              (bind _output_get (LocalVariables|SetAttachmentRule (select (Math|Vector|VectorIsNearlyZero _returnvalue) "KeepRelative" "KeepWorld")))
              (bind _returnvalue_2 (Transformation|AttachActorToComponent (Variables|Donut|GetSpawnedProp) (Variables|Transformation|GetRootComponent) "None" _output_get _output_get _output_get true))
              (if _returnvalue_2
                (Fortnite|FortShippingLog self (Utilities|String|Append "[Client] Attached donut-actor to drop-actor.  Loc before [" (Utilities|String|ToString(Vector) (LocalVariables|GetSpawnedPropPrevLoc)) "] after [" (Utilities|String|ToString(Vector) (Transformation|GetActorLocation (Variables|Donut|GetSpawnedProp))) "] using rule [" (Utilities|String|EnumtoString (LocalVariables|GetAttachmentRule)) "]"))
                (else
                  (Fortnite|FortShippingLog self "[Client] FAILED to attach donut-actor to drop-actor" true) ; Native typically logs an error when this occurs.\nLog an error in BP, too, so we know which BP function the request came from.
                ))))
          (:Is%20Not%20Valid
            (Development|PrintString "[Client] SpawnedPropMesh is null")))))))

(fn GetBottomZ
  (bind (_returnvalue_x _returnvalue_y _returnvalue_z) (Math|Vector|BreakVector (Transformation|GetActorLocation)))
  (return (- _returnvalue_z (Math|Float|Max(Float) (Variables|Movement|GetBottomOffsetZ)))))

(fn GetFinalLocation
  (bind (_groundlocation_x _groundlocation_y) (Math|Vector|BreakVector (Default|GetGroundLocation)))
  (bind (_groundlocation_x_2 _groundlocation_y_2 _groundlocation_z) (Math|Vector|BreakVector (Default|GetGroundLocation)))
  (LocalVariables|SetFinalZ (+ _groundlocation_z (Math|Float|Max(Float) (Variables|Movement|GetFinalOffsetZ))))
  (if (not (Variables|Default|GetbAllowHovering))
    (LocalVariables|SetFinalZ (Math|Float|Min(Float) (LocalVariables|GetFinalZ) (+ _groundlocation_z (Math|Float|Max(Float) (Variables|Movement|GetBottomOffsetZ))))) ; Clamp the donut downward so that it will touch the ground
  )
  (return _groundlocation_x _groundlocation_y (LocalVariables|GetFinalZ)))

(fn GetGroundLocation
  (bind (_returnvalue_x _returnvalue_y) (Math|Vector|BreakVector (Transformation|GetActorLocation)))
  (return _returnvalue_x _returnvalue_y (Variables|Ground|GetGroundZ)))

(fn GetReachedGroundZ
  (bind (_groundlocation_x _groundlocation_y _groundlocation_z) (Math|Vector|BreakVector (Default|GetGroundLocation)))
  (return (+ _groundlocation_z (Math|Float|Max(Float) (Variables|Movement|GetBottomOffsetZ)))))

(fn GetReachedWaterZ
  (bind (_waterlocation_x _waterlocation_y _waterlocation_z) (Math|Vector|BreakVector (Default|GetWaterLocation)))
  (return (+ _waterlocation_z (Math|Float|Max(Float) (Variables|Movement|GetBottomOffsetZ)))))

(fn GetSpawnedPropMesh
  (bind _spawnedprop (Variables|Donut|GetSpawnedProp)
    (:then
      (return (Variables|/Script/FortniteGame.BuildingSMActor|GetStaticMeshComponent _spawnedprop)))
    (:else
      (return 0))))

(fn GetWaterLocation
  (bind (_returnvalue_x _returnvalue_y) (Math|Vector|BreakVector (Transformation|GetActorLocation)))
  (return _returnvalue_x _returnvalue_y (Variables|Water|GetWaterZ)))

(fn HasStoppedMovement
  (Utilities|FlowControl|Sequence
    (:then_0
      (bind _fortprojectilemovement (Variables|Default|GetFortProjectileMovement)
        (:then
          (if (Components|Activation|IsActive _fortprojectilemovement)
            (Variables|/Script/Engine.MovementComponent|GetUpdatedComponent _fortprojectilemovement
              (:then
                (return false))
              (:else))))
        (:else)) ; Return true if we still have a movement component, it's active,\nand its UpdatedComponent is valid.\nUpdatedComponent gets cleared in native StopSimulating().
    )
    (:then_1
      (return true))))

(fn OnRep_bShouldStopEverything
  (CallFunction|StopEverything))

(fn OnRep_FinalRotation
  (Networking|SwitchHasAuthority
    (:Remote
      (Transformation|SetActorRotation (Variables|Default|GetFinalRotation) true) ; Synchronize client rotation with the server
      (Variables|Default|SetbHasClientAppliedFinalRotation true)
      (bind _rotatingmovement (Variables|Default|GetRotatingMovement)
        (:then
          (Components|Activation|Deactivate _rotatingmovement) ; Prevent further rotation
        )
        (:else)))))

(fn OnRep_SpawnedProp
  (bind _static_mesh_component (CallFunction|GetSpawnedPropMesh))
  (Networking|CanExecuteCosmeticEvents
    (:True
      (Utilities|IsValid _static_mesh_component
        (:Is%20Valid
          (Variables|Wobble|SetPropStartingScale (Transformation|GetWorldScale _static_mesh_component)) ; Cache the starting scale of the donut\nprop mesh.  We don't expect any server\nlogic to modify scale, so we don't have\nto worry about the replicated scale\nbeing dependent on the timing of when\nthe donut actor becomes relevant.\nIt should ALWAYS be constant on the server.
          (Variables|Default|SetbHasPropStartingScaleBeenCached true) ; The intro and wobble timelines cannot run\nunless this bool has been set
          (CallFunction|DoIntroScale)
          (if (CallFunction|HasStoppedMovement)
            (CallFunction|SetMobilitytoStatic) ; The attached donut Mobility isn't replicated,\nso we must also set it on client
            (if (Variables|Default|GetbShouldAutoCorrectClientAttachment)
              (Utilities|Time|SetTimerforNextTickbyFunctionName 0 "FixClientAttachment") ; Wait one frame before attempting to fix client attachment.\nIt will hopefully be dealt with in AActor::OnRep_AttachmentReplication() before then.
            )))))))

(fn ProcessNonBuildingHit
  (params (HitActor :type Actor%20Object%20Reference) (DamageVehicleGE :type Gameplay%20Effect%20Class%20Reference) (DamagePawnGE :type Gameplay%20Effect%20Class%20Reference) (DamagePhysicsObjectGE :type Gameplay%20Effect%20Class%20Reference))
  (bind _returnvalue (Transformation|GetActorLocation))
  (bind _knockaddedz (Variables|PawnHits|GetKnockAddedZ))
  (bind _knockmag (Variables|PawnHits|GetKnockMag))
  (if (not (Utilities|Array|ContainsItem (Variables|PawnHits|GetIgnoreNonBuildings) HitActor))
    (Utilities|Array|Add (Variables|PawnHits|GetIgnoreNonBuildings) (Variables|GetHitActor))
    (bind _asfort_vehicle_interface (Utilities|Casting|CastToFortVehicleInterface HitActor)
      (:then
        (bind _returnvalue_2 (Utilities|MakeMyFortContext))
        (GameplayEffects|ApplyGameplayEffectToTarget (Ability|GetAbilitySystemComponent self) (Variables|GetDamageVehicleGE) (Vehicle|GetVehicleAbilitySystemComponent _asfort_vehicle_interface) 0.0 _returnvalue_2))
      (:CastFailed
        (bind _asfort_pawn (Utilities|Casting|CastToFortPawn HitActor)
          (:then
            (bind _asfort_player_pawn (Utilities|Casting|CastToFortPlayerPawn (Variables|GetHitActor))
              (:then
                (bind (_returnvalue_x _returnvalue_y) (Math|Vector|BreakVector (Math|Vector|Normalize (- (Transformation|GetActorLocation _asfort_pawn) _returnvalue))))
                (bind (_returnvalue_x_2 _returnvalue_y_2 _returnvalue_z) (Math|Vector|BreakVector (Math|Vector|vector*float (Math|Vector|MakeVector _returnvalue_x _returnvalue_y) _knockmag)))
                (Utilities|LaunchPlayerPawn (Utilities|Casting|CastToFortPlayerPawn (Game|GetInstigator)) _asfort_player_pawn (Math|Vector|MakeVector _returnvalue_x_2 _returnvalue_y_2 (+ _returnvalue_z _knockaddedz)) true true false true true true true)
                (bind _returnvalue_3 (Utilities|MakeMyFortContext))
                (GameplayEffects|ApplyGameplayEffectToTarget (Ability|GetAbilitySystemComponent self) (Variables|GetDamagePawnGE) (Ability|GetAbilitySystemComponent (Variables|GetHitActor)) 0.0 _returnvalue_3))
              (:CastFailed
                (Character|LaunchCharacter _asfort_pawn (* (Math|Vector|MakeVector _returnvalue_x_2 _returnvalue_y_2 (+ _returnvalue_z _knockaddedz)) (Variables|PawnHits|GetDeimosImpulseMult)) true true))))
          (:CastFailed
            (Variables|GetDamagePhysicsObjectGE
              (:then
                (if (Utilities|IsValid (Actor|GetComponentbyClass HitActor "/Script/FortniteGame.FortPhysicsObjectComponent") :!class /Script/Engine.KismetSystemLibrary)
                  (bind _returnvalue_4 (Utilities|MakeMyFortContext))
                  (GameplayEffects|ApplyGameplayEffectToTarget (Ability|GetAbilitySystemComponent self) (Variables|GetDamagePhysicsObjectGE) (Ability|GetAbilitySystemComponent HitActor) 0.0 _returnvalue_4)))
              (:else))))))))

(fn SetMobilityToStatic
  (bind _spawnedpropmesh (LocalVariables|GetSpawnedPropMesh))
  (bind _mobility (Variables|/Script/Engine.SceneComponent|GetMobility _spawnedpropmesh))
  (bind _rootscene (Variables|Default|GetRootScene))
  (bind _mobility_2 (Variables|/Script/Engine.SceneComponent|GetMobility _rootscene))
  (Utilities|FlowControl|Sequence
    (:then_0
      (if (Utilities|Enum|NotEqual(Enum) "Static" _mobility_2)
        (Development|PrintString (Utilities|String|Append (Utilities|String|SelectString "[Server]" "[Client]" (Networking|IsServer)) " Changing drop-actor mobility from [" (Utilities|String|EnumtoString _mobility_2) "] to [Static]"))
        (Transformation|SetMobility (Variables|Default|GetRootScene) "Static")))
    (:then_1
      (bind _output_get (LocalVariables|SetSpawnedPropMesh (CallFunction|GetSpawnedPropMesh)))
      (Utilities|IsValid _output_get
        (:Is%20Valid
          (if (Utilities|Enum|NotEqual(Enum) "Static" _mobility)
            (Development|PrintString (Utilities|String|Append (Utilities|String|SelectString "[Server]" "[Client]" (Networking|IsServer)) " Changing donut-actor mobility from [" (Utilities|String|EnumtoString _mobility) "] to [Static]"))
            (Transformation|SetMobility (LocalVariables|GetSpawnedPropMesh) "Static")))
        (:Is%20Not%20Valid
          (Fortnite|FortShippingLog self "[Server] SpawnedPropMesh is null" true))))))

(fn ConstructionScript)
