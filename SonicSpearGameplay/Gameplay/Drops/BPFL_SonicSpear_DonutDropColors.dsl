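; Blueprint function library function that picks a coloring for Sonic Spear donut drops.
; When data table sets are enabled and the table is valid, it returns a uniformly random row with a success flag.
; A failed row lookup returns false, while a disabled or invalid table returns zeroed colors and false.
; A TODO notes a planned ruleset to choose values in place of the data table.
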(fn ChooseSonicSpearDropColoring
  (params (Use%20Data%20Table%20Sets :type Boolean :default true) (Data%20Table :type Data%20Table%20Object%20Reference))
  (if Use%20Data%20Table%20Sets
    (if (Utilities|IsValid Data%20Table :!class /Script/Engine.KismetSystemLibrary)
      (bind _outrownames (DataTable|GetDataTableRowNames Data%20Table))
      (bind _returnvalue (Utilities|GetDataTableRow (Variables|GetData%20Table) (Utilities|Array|Get(acopy) _outrownames (Math|Random|RandomIntegerinRange 0 (Utilities|Array|LastIndex _outrownames))))
        (:then
          (return _returnvalue true))
        (:RowNotFound
          (return _returnvalue false)))
      (else
        _))
    (else
      _))
  ;[0.15 0.002691 0 0.5] TODO setup ruleset to choose values instead of data table
  (return "(R=0.000000,G=0.000000,B=0.000000,A=0.000000)" "(R=0.000000,G=0.000000,B=0.000000,A=0.000000)" false))
