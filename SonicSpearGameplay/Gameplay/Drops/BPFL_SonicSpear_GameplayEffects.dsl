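; Blueprint function library for the Sonic Spear that applies damage gameplay effects and knockback to a hit actor.
; A cast chain picks vehicle, pawn or physics-object handling, sourcing effects from TargetActor's ability system through a shared context helper.
; Player pawns are launched away from TargetActor with added lift, then damaged.
; Other pawns only get a multiplier-scaled launch, reusing vector math from the player branch.
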(fn SonicSpearMakeEffectContext
  (params (Actor :type Actor%20Object%20Reference))
  (bind _asfort_ability_system_component (Utilities|Casting|CastToFortAbilitySystemComponent (Ability|GetAbilitySystemComponent Actor)))
  (return (GameplayEffects|MakeFortEffectContext _asfort_ability_system_component 0 0 Actor 0)))

(fn SonicSpearSpawnDamageGameplayEffect
  (params (TargetActor :type Actor%20Object%20Reference) (HitActor :type Actor%20Object%20Reference) (DamageVehicleGE :type Gameplay%20Effect%20Class%20Reference) (DamagePawnGE :type Gameplay%20Effect%20Class%20Reference) (DamagePhysicsObjectGE :type Gameplay%20Effect%20Class%20Reference) (HorizontalPawnImpulseMultiplier :type Float%20%28double-precision%29) (VerticalPawnImpulse :type Float%20%28double-precision%29) (DeimosImpulseMultiplier :type Float%20%28double-precision%29))
  (bind _targetactor (Variables|GetTargetActor))
  (bind _returnvalue (Transformation|GetActorLocation _targetactor))
  (bind _hitactor (Variables|GetHitActor))
  (bind _horizontalpawnimpulsemultiplier (Variables|GetHorizontalPawnImpulseMultiplier))
  (bind _verticalpawnimpulse (Variables|GetVerticalPawnImpulse))
  ;[0 0 0 0.5] Vehicle Damage
  (bind _asfort_vehicle_interface (Utilities|Casting|CastToFortVehicleInterface (Variables|GetHitActor))
    (:then
      (Variables|GetDamageVehicleGE
        (:then
          (GameplayEffects|ApplyGameplayEffectToTarget (Ability|GetAbilitySystemComponent (Variables|GetTargetActor)) (Variables|GetDamageVehicleGE) (Vehicle|GetVehicleAbilitySystemComponent _asfort_vehicle_interface) 0.0 (CallFunction|SonicSpearMakeEffectContext (Variables|GetTargetActor) (Variables|GetTargetActor))))
        (:else)))
    (:CastFailed
      ;[0 0 0 0.5] Pawn Damage
      (bind _asfort_pawn (Utilities|Casting|CastToFortPawn (Variables|GetHitActor))
        (:then
          (Variables|GetDamagePawnGE
            (:then
              (bind _asfort_player_pawn (Utilities|Casting|CastToFortPlayerPawn (Variables|GetHitActor))
                (:then
                  (bind (_returnvalue_x _returnvalue_y) (Math|Vector|BreakVector (Math|Vector|Normalize (- (Transformation|GetActorLocation _asfort_pawn) _returnvalue))))
                  (bind (_returnvalue_x_2 _returnvalue_y_2 _returnvalue_z) (Math|Vector|BreakVector (Math|Vector|vector*float (Math|Vector|MakeVector _returnvalue_x _returnvalue_y) _horizontalpawnimpulsemultiplier)))
                  (Utilities|LaunchPlayerPawn (Utilities|Casting|CastToFortPlayerPawn (Game|GetInstigator (Variables|GetTargetActor))) _asfort_player_pawn (Math|Vector|MakeVector _returnvalue_x_2 _returnvalue_y_2 (+ _returnvalue_z _verticalpawnimpulse)) true true false true true true true)
                  (GameplayEffects|ApplyGameplayEffectToTarget (Ability|GetAbilitySystemComponent (Variables|GetTargetActor)) (Variables|GetDamagePawnGE) (Ability|GetAbilitySystemComponent (Variables|GetHitActor)) 0.0 (CallFunction|SonicSpearMakeEffectContext (Variables|GetTargetActor) (Variables|GetTargetActor))))
                (:CastFailed
                  (Character|LaunchCharacter _asfort_pawn (* (Math|Vector|MakeVector _returnvalue_x_2 _returnvalue_y_2 (+ _returnvalue_z _verticalpawnimpulse)) (Variables|GetDeimosImpulseMultiplier)) true true))))
            (:else)))
        (:CastFailed
          ;[0 0 0 0.5] Physics Object Damage
          (Variables|GetDamagePhysicsObjectGE
            (:then
              (if (Utilities|IsValid (Actor|GetComponentbyClass _hitactor "/Script/FortniteGame.FortPhysicsObjectComponent") :!class /Script/Engine.KismetSystemLibrary)
                (GameplayEffects|ApplyGameplayEffectToTarget (Ability|GetAbilitySystemComponent (Variables|GetTargetActor)) (Variables|GetDamagePhysicsObjectGE) (Ability|GetAbilitySystemComponent _hitactor) 0.0 (CallFunction|SonicSpearMakeEffectContext (Variables|GetTargetActor) (Variables|GetTargetActor)))))
            (:else)))))))
