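; BP_SonicSpear_Donut is a destructible Sonic Spear donut with colored frosting and an optional golden variant.
; On begin play, the server can load a preset color from a DataTable, and cosmetic-capable instances write it to custom primitive data.
; Death spawns color- and scale-matched Niagara destruction with sound, while a separate event spawns matching debris at the impact location.
; The construction script previews non-preset colors in editor; a custom-depth workaround is flagged TODO.
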(event EventBeginPlay
  (bind (_colordatatable_datatable _colordatatable_rowname) (Utilities|Struct|BreakDataTableRowHandle (Variables|GetColorDataTable_DataTable)))
  (|CallParentFunction)
  (Utilities|FlowControl|Sequence
    (:then_0
      ;[0 0.02781 0.15 0.5] initialize frosting color on the donuts
      (if (Variables|Override|GetbUseColorPreset)
        (Networking|SwitchHasAuthority
          (:Authority
            (bind _returnvalue (Utilities|GetDataTableRow _colordatatable_datatable _colordatatable_rowname))
            (Variables|Default|SetColoring _returnvalue))) ; set the color from server
      ))
    (:then_1
      (CallFunction|UpdateFrostingColor)
      (Rendering|SetRenderCustomDepth (Variables|Debug|GetStaticMeshComponent) false)
      ;[1 1 1] Reenables Outlines
      ; TODO: Why is this necessary?
      (Utilities|FlowControl|DelayUntilNextTick)
      (Rendering|SetRenderCustomDepth (Variables|Debug|GetStaticMeshComponent) false))))

(event EventOnDeathPlayEffects
  (params (Damage :type Float%20%28single-precision%29) (DamageTags :type Gameplay%20Tag%20Container%20Structure%20%28by%20ref%29) (Momentum :type Vector :default "0, 0, 0") (HitInfo :type Hit%20Result%20Structure%20%28by%20ref%29) (InstigatedBy :type Fort%20Pawn%20Object%20Reference) (DamageCauser :type Actor%20Object%20Reference) (EffectContext :type Gameplay%20Effect%20Context%20Handle%20Structure))
  (bind (_returnvalue_location _returnvalue_rotation _returnvalue_scale) (Math|Transform|BreakTransform (Transformation|GetWorldTransform (Variables|Debug|GetStaticMeshComponent))))
  (bind (_returnvalue_scale_x _returnvalue_scale_y _returnvalue_scale_z) (Math|Vector|BreakVector _returnvalue_scale))
  (bind _staticmeshcomponent (Variables|Debug|GetStaticMeshComponent))
  (bind _returnvalue (Transformation|GetWorldLocation _staticmeshcomponent))
  (bind _coloring_frostingcoloring_15_ed6d36ac450de5d5c052ada0a0a175be (Utilities|Struct|BreakSSonicSpearDonutDropColor (Variables|GetColoring_FrostingColoring_15_ED6D36AC450DE5D5C052ADA0A0A175BE)))
  ;[1 1 1] Coloring
  (|CallParentFunction Damage DamageTags Momentum HitInfo InstigatedBy DamageCauser EffectContext)
  (bind _returnvalue_2 (Niagara|SpawnSystematLocation "/SonicSpearGameplay/Assets/VFX/NS_SonicSpear_Destruction.NS_SonicSpear_Destruction" _returnvalue (Transformation|GetWorldRotation _staticmeshcomponent)))
  (Effects|Components|ParticleSystem|SetColorParameter _returnvalue_2 "Frosting" _coloring_frostingcoloring_15_ed6d36ac450de5d5c052ada0a0a175be)
  (Effects|Components|ParticleSystem|SetFloatParameter _returnvalue_2 "Size" (- (* _returnvalue_scale_z 2.0) 1.0))
  (Transformation|SetRelativeScale3D _returnvalue_2 "1.000000,1.000000,1.000000")
  (Effects|Components|ParticleSystem|SetBoolParameter _returnvalue_2 "Golden" (Variables|Default|GetGolden))
  (Audio|PlaySoundatLocation (Variables|Effects|GetDeathSound) _returnvalue))

(event Custom|SpawnDebris
  (bind (_returnvalue_x _returnvalue_y _returnvalue_z) (Math|Vector|BreakVector (Transformation|GetWorldScale (Variables|Debug|GetStaticMeshComponent))))
  (Game|SpawnActorfromClass (Variables|Default|GetDebrisClass) (Math|Transform|MakeTransform (Variables|Default|GetImpactLocation) (Transformation|GetActorRotation)) "AlwaysSpawn" "MultiplyWithRoot" 0 (Variables|Default|GetColoring) _returnvalue_z self 0))

(fn UpdateFrostingColor
  (Networking|CanExecuteCosmeticEvents
    (:True
      (Rendering|Material|SetCustomPrimitiveDataVector4 (Variables|Debug|GetStaticMeshComponent) 0 (Class|CreativeColorComponentBlueprintLibrary|RGBAToVec4 (Utilities|Struct|BreakSSonicSpearDonutDropColor (Variables|Default|GetColoring)))))))

(fn ConstructionScript
  ;[0.15 0.1097 0 0.5] show proper frosting colors for donuts pre-placed in levels (in editor only)
  (if (and (not (Variables|Override|GetbUseColorPreset)) (Editor|InEditor))
    (CallFunction|UpdateFrostingColor)))
