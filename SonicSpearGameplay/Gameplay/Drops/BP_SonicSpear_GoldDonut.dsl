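; Gold Donut drop Blueprint for SonicSpear that spawns and tosses loot pickups on its server death event.
; SpawnAndTossLoot, gated by a scalable-float toggle, rolls a tiered loot table above the spawn location, tags pickups with source tags and tosses them around 360 degrees.
; The LootSpawned flag prevents repeat spawns.
; The authority-only Destroyed path calls it only when that flag is already set, so that path spawns nothing.
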(event EventDestroyed
  (Utilities|FlowControl|DoOnce false
    (:Completed
      (Networking|SwitchHasAuthority
        (:Authority
          (if (Variables|Default|GetLoot%20Spawned)
            (CallFunction|SpawnandTossLoot)))))))

(event EventOnDeathServer
  (params (Damage :type Float%20%28single-precision%29) (DamageTags :type Gameplay%20Tag%20Container%20Structure%20%28by%20ref%29) (Momentum :type Vector :default "0, 0, 0") (HitInfo :type Hit%20Result%20Structure%20%28by%20ref%29) (InstigatedBy :type Controller%20Object%20Reference) (DamageCauser :type Actor%20Object%20Reference) (EffectContext :type Gameplay%20Effect%20Context%20Handle%20Structure))
  (Utilities|FlowControl|DoOnce false
    (:Completed
      (CallFunction|SpawnandTossLoot))))

(fn GetLootSpawnPosition
  (return (+ (Transformation|GetWorldLocation (Variables|Default|GetItem%20Spawn%20Location)) (* (Transformation|GetActorUpVector) 5e+01))))

(fn OnRep_LootSpawned)

(fn SpawnAndTossLoot
  (bind (_returnvalue_roll _returnvalue_pitch _returnvalue_yaw) (Math|Rotator|BreakRotator (Math|Rotator|MakeRotfromX (Transformation|GetActorUpVector))))
  (if (not (Variables|Default|GetLoot%20Spawned))
    (Utilities|FlowControl|Sequence
      (:then_0
        (if (ScalableFloat|GetValueasBool (Variables|Default|GetSpawnFromLootTables))
          ;[0 0.1062 0.15 0.5] Spawn Loot
          (bind _itempickupsspawned (Class|AthenaFunctionLibrary|SpawnLootfromTableUsingItemEntry (Variables|Default|GetLootTierToSpawn) 0 4 (CallFunction|GetLootSpawnPosition) "-1.000000,0.000000,0.000000" 0 true true false 0 "Container" "Chest"))
          (bind _output_get (LocalVariables|SetSpawned%20Loot _itempickupsspawned))
          (for (_array_element _array_index) _output_get
            (Class|FortPickup|AddSpawnSourceTags _array_element (Variables|Default|GetLoot%20Source%20Tags))
            (Fortnite|Container|TossPickupfromContainer self _array_element (Utilities|Array|LastIndex _output_get) _array_index 3.6e+02 (Math|Rotator|MakeRotator 0.0 35.0 _returnvalue_yaw) (Math|Random|RandomFloatinRange 6e+02 8e+02) false false 0))))
      (:then_1
        (Variables|Default|SetLoot%20Spawned true)))))

(fn ConstructionScript)
