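; Small horizontal SonicSpear treasure chest that players search once to spawn loot from its loot tier table.
; Interaction sets bLootClaimed and tosses tagged pickups along the up vector's yaw; destroying an unclaimed chest also drops loot on authority.
; OnRep_bLootClaimed hides the glow, fades audio and plays an opening emitter plus timeline-driven material timestamp.
; Construction script configures smoke and glow through custom primitive data.
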(event Interaction|EventBlueprintOnInteract
  (params (InteractingPawn :type Fort%20Pawn%20Object%20Reference) (InteractionBeingAttempted :type EInteractionBeingAttempted%20Enum :default "FirstInteraction"))
  (Variables|Default|SetbLootClaimed true)
  (CallFunction|SpawnandTossLoot))

(event Custom|SetChestClaimedVisuals
  (Development|SetHiddeninGame (Variables|Default|GetChest%20Smoke%20and%20Glow) true false)
  (Audio|Components|Audio|FadeOut (Variables|Default|GetAmbientAudio) 0.2 0.0 "Linear" :!class /Script/Engine.AudioComponent)
  (CallFunction|PlayChestOpening))

(event Custom|PlayChestOpening
  (bind _p_loot_chest_opened_fx (Variables|Default|GetP_Loot_Chest_Opened_FX))
  (Utilities|FlowControl|DoOnce false
    (:Completed
      (Utilities|IsValid _p_loot_chest_opened_fx
        (:Is%20Valid
          (Effects|SpawnEmitteratLocation _p_loot_chest_opened_fx (Transformation|GetWorldLocation (Variables|Default|GetChest_Mark_III)))
          (bind (_direction _time) (Timeline|Timeline Play)
            (:Update
              (Rendering|Material|SetScalarParameterValueonMaterials (Variables|Default|GetChest_Mark_III) "CODE_Timestamp" _time))))))))

(event Custom|SpawnAndTossLoot
  (bind (_returnvalue_roll _returnvalue_pitch _returnvalue_yaw) (Math|Rotator|BreakRotator (Math|Rotator|MakeRotfromX (Transformation|GetActorUpVector))))
  (bind _itempickupsspawned (Class|AthenaFunctionLibrary|SpawnLootfromTableUsingItemEntry (Variables|Default|GetLootTierToSpawn) 0 -1 (+ (Transformation|GetWorldLocation (Variables|Default|GetChest_Mark_III)) (* (Transformation|GetActorUpVector) 5e+01)) "-1.000000,0.000000,0.000000" 0 true true false 0 "Container" "Chest"))
  (for (_array_element _array_index) _itempickupsspawned
    ;[0.3166 0.4028 1] Use a more authored toss because the standard one is unreliable
    (Class|FortPickup|AddSpawnSourceTags _array_element (Variables|Default|GetLoot%20Source%20Tags))
    (Fortnite|Container|TossPickupfromContainer self _array_element (- (Utilities|Array|Length _itempickupsspawned) 1) _array_index 3.6e+02 (Math|Rotator|MakeRotator 0.0 35.0 _returnvalue_yaw) (Math|Random|RandomFloatinRange 6e+02 8e+02) false false 0)))

(event EventDestroyed
  (Utilities|FlowControl|DoOnce false
    (:Completed
      (Networking|SwitchHasAuthority
        (:Authority
          (if (not (Variables|Default|GetbLootClaimed))
            (CallFunction|SpawnandTossLoot)))))))

(fn BlueprintCanInteract
  (params (InteractingPawn :type Fort%20Pawn%20Object%20Reference) (InteractionBeingAttempted :type EInteractionBeingAttempted%20Enum :default "FirstInteraction") (InteractionType :type TInteractionType%20Enum :default "IT_NoInteraction"))
  (return (not (Variables|Default|GetbLootClaimed))))

(fn BlueprintGetInteractionString
  (params (InteractingPawn :type Fort%20Pawn%20Object%20Reference) (InteractionBeingAttempted :type EInteractionBeingAttempted%20Enum :default "FirstInteraction"))
  (return "NSLOCTEXT(\"[5A6B01D59A904602E88233D57169CE0E]\", \"1BAB788A4DC35C18BF8A7F96949F7370\", \"Search\")"))

(fn FXSmokeSheetInitialization
  (bind _chest_smoke_and_glow (Variables|Default|GetChest%20Smoke%20and%20Glow))
  (Utilities|FlowControl|Sequence
    (:then_0
      (Variables|Default|SetLastShouldHideSmokeSheetSparkles (Variables|Default|GetShouldHideSmokeSheetSparkles))
      (Rendering|Material|SetDefaultCustomPrimitiveDataFloat _chest_smoke_and_glow 0 (Math|Float|SelectFloat 1.0 0.0 (Variables|Default|GetShouldHideSmokeSheetSparkles))) ; WPO_HideMesh
      (Rendering|Material|SetDefaultCustomPrimitiveDataFloat _chest_smoke_and_glow 2 (Math|Float|SelectFloat 0.0 1.0 (Variables|Default|GetShouldHideSmokeSheetSparkles))) ; Material Opacity
    )
    (:then_1
      (Variables|Default|SetLast%20Glow%20Corona%20Size%20Scale (Variables|Default|GetGlow%20Corona%20Size%20Scale))
      (Rendering|Material|SetDefaultCustomPrimitiveDataFloat (Variables|Default|GetChest%20Smoke%20and%20Glow) 1 (Variables|Default|GetGlow%20Corona%20Size%20Scale)) ; Scaled_Brightness
    )))

(fn OnRep_bLootClaimed
  (Networking|CanExecuteCosmeticEvents
    (:True
      (CallFunction|SetChestClaimedVisuals))))

(fn ConstructionScript
  (|CallParentFunction)
  (Networking|CanExecuteCosmeticEvents
    (:True
      (CallFunction|FXSmokeSheetInitialization))))
