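; Small SonicSpear mystery drop that wraps a WildEstate faction chest, held in the MysteryChest child actor component, and hands out its loot.
; Interacting casts the child to the chest container, spawns loot for the interacting player and stores the result in LootClaimed.
; If destroyed unclaimed, the authority spawns the loot once with no player.
; OnRep_LootClaimed plays the chest-open VFX where cosmetic events can run.
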(event Interaction|EventBlueprintOnInteract
  (params (InteractingPawn :type Fort%20Pawn%20Object%20Reference) (InteractionBeingAttempted :type EInteractionBeingAttempted%20Enum :default "FirstInteraction"))
  (bind _asfort_player_pawn (Utilities|Casting|CastToFortPlayerPawn InteractingPawn))
  (bind _asb_wild_estate_faction_chest_container (Utilities|Casting|CastToB_WildEstate_FactionChest_Container (Variables|/Script/Engine.ChildActorComponent|GetChildActor (Variables|Default|GetMysteryChest))))
  (bind _returnvalue (Container|SpawnLoot _asb_wild_estate_faction_chest_container _asfort_player_pawn))
  (Variables|Default|SetLootClaimed _returnvalue))

(event EventDestroyed
  (Utilities|FlowControl|DoOnce false
    (:Completed
      (Networking|SwitchHasAuthority
        (:Authority
          (if (not (Variables|Default|GetLootClaimed))
            (bind _asb_wild_estate_faction_chest_container (Utilities|Casting|CastToB_WildEstate_FactionChest_Container (Variables|/Script/Engine.ChildActorComponent|GetChildActor (Variables|Default|GetMysteryChest))))
            (bind _returnvalue (Container|SpawnLoot _asb_wild_estate_faction_chest_container 0))
            (Variables|Default|SetLootClaimed _returnvalue)))))))

(fn OnRep_LootClaimed
  (Networking|CanExecuteCosmeticEvents
    (:True
      (bind _asb_wild_estate_faction_chest_container (Utilities|Casting|CastToB_WildEstate_FactionChest_Container (Variables|/Script/Engine.ChildActorComponent|GetChildActor (Variables|Default|GetMysteryChest))))
      (Class|TieredChestAthenaParent|SpawnChestOpenVFX _asb_wild_estate_faction_chest_container))))

(fn ConstructionScript
  (|CallParentFunction))
