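; Ground target marker for a falling Sonic Spear drop, showing a decal and optional post-process warning at its location.
; Server strips client-only components, replicates decal scale derived from the falling object, and traces to align to the surface, hiding on a miss.
; Clients enable box collision and run a timer feeding falling-object distance into the decal's dynamic material.
; OnLanded, bound on server and via OnReps, hides the actor and sets its lifespan.
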(event EventBeginPlay
  (bind (_relativescale3d_x _relativescale3d_y _relativescale3d_z) (Math|Vector|BreakVector (Variables|/Script/Engine.SceneComponent|GetRelativeScale3D_X (Variables|Default|GetFallingObj))))
  ;[0 1 0 0.5] [BeginPlay]
  (Networking|SwitchHasAuthority
    (:Authority
      ;;[1 1 1] Server setup
      (Variables|Default|GetFallingObj
        (:then
          (Components|DestroyComponent (Variables|Default|GetPostProcess)) ; Destroy components\nthat the server doesn't use
          (Components|DestroyComponent (Variables|Default|GetAmbientAudio))
          (Utilities|FlowControl|Sequence
            (:then_0
              ;;;[0 0 0 0.5] Monitor for the DropActor'slanding, so we know when to self-destruct. Account for standard and physics variants.
              (bind _dropactor (Variables|Default|GetDropActor)
                (:then
                  (Default|BindEventtoOnLanded _dropactor (EventDispatchers|CreateEvent :!event Custom|DropActorLanded) :!class /SonicSpearGameplay/Gameplay/Drops/BGA_SonicSpear_MainDropParent.BGA_SonicSpear_MainDropParent_C))
                (:else
                  (bind _physicsdropactor (Variables|Default|GetPhysicsDropActor)
                    (:then
                      (Default|BindEventtoOnLanded _physicsdropactor (EventDispatchers|CreateEvent :!event Custom|DropActorLanded) :!class /SonicSpearGameplay/Gameplay/Drops/B_SonicSpear_Drop_Physics_Parent.B_SonicSpear_Drop_Physics_Parent_C))
                    (:else
                      (Fortnite|FortShippingLog self "DropActor is null.  Self-destruct.")
                      (Actor|DestroyActor) ; No reason to stick around if\nwe're lacking the supporting objects
                    )))))
            (:then_1
              (if (not (Game|IsActorBeingDestroyed))
                (Variables|Default|SetDecalRelativeScale (Math|Vector|SelectVector (Variables|Default|GetCustomScale) (Math|Vector|Lerp(Vector) "0.500000,0.500000,0.500000" "1.000000,1.000000,1.000000" (- (* _relativescale3d_z 2.0) 1.0)) (Variables|Default|GetbUseCustomScale))) ; Compute and replicate the desired decal\nscale to the client, which is based on FallingObj.\nThe client may not have the ref to FallingObj\ndue to relevancy.
                (bind (_outhit _returnvalue) (Collision|LineTraceForObjects :Start (Utilities|Operators|Add (Transformation|GetActorLocation) (Math|Vector|MakeVector 0.0 0.0 (Variables|Default|GetStartTraceOffsetZ))) :End (Utilities|Operators|Add (Transformation|GetActorLocation) (Math|Vector|MakeVector 0.0 0.0 (Variables|Default|GetEndTraceOffsetZ))) :ObjectTypes (Variables|Default|GetTraceObjectTypes) :bTraceComplex false :DrawDebugType "None" :bIgnoreSelf true :TraceColor "(R=1.000000,G=0.000000,B=0.000000,A=1.000000)" :TraceHitColor "(R=0.000000,G=1.000000,B=0.000000,A=1.000000)" :DrawTime 1e+01)) ; Perform a trace to try to find a surface to align to
                (if _returnvalue
                  (Transformation|SetActorRotation (Math|Rotator|MakeRotfromZ (.ImpactNormal (Collision|BreakHitResult _outhit)))) ; Rotate to match the ImpactNormal
                  (else
                    (Rendering|SetActorHiddenInGame true) ; No hit found.  Rather than risking a hovering\ndecal, just hide this actor.  bHidden is replicated.\nStay alive for the purpose of the PostProcess.  
                  ))))))
        (:else
          (Fortnite|FortShippingLog self "FallingObj is null.  Self-destruct.")
          (Actor|DestroyActor) ; No reason to stick around if\nwe're lacking the supporting objects
        )))
    (:Remote
      (Utilities|FlowControl|Sequence
        (:then_0
          ;;[1 1 1] Client setup
          (Transformation|SetRelativeScale3D (Variables|Default|GetDecal) (Variables|Default|GetDecalRelativeScale)) ; Set the Decal's scale using the vector\nthat was determined by the server\nand replicated to us
          (bind _returnvalue_2 (Rendering|Material|CreateDynamicMaterialInstance (Variables|/Script/Engine.DecalComponent|GetDecalMaterial (Variables|Default|GetDecal)) "None" "None" :!class /Script/Engine.KismetMaterialLibrary))
          (bind _output_get (Variables|Default|SetDecalMID _returnvalue_2))
          (bind _outparameterindex (Rendering|Material|InitializeScalarParameterandGetIndex _output_get (Utilities|Name|MakeLiteralName "PercentDistanceFromTarget") (Variables|Default|GetTargetPercentDistanceFromDropActor)))
          (Variables|Default|SetMaterialDistancePercentageParameterIndex _outparameterindex)
          (Rendering|Components|Decal|SetDecalMaterial (Variables|Default|GetDecal) _output_get))
        (:then_1
          (Variables|/Script/Engine.PostProcessComponent|SetbEnabled (ScalableFloat|GetValueasBool (Variables|Default|GetHFRow_PostProcessWarningEnabled)) (Variables|Default|GetPostProcess))
          (Collision|SetCollisionEnabled (Variables|Default|GetBox) "QueryOnly") ; Box collision is disabled by default, so that it doesn't\nimmediately trigger overlaps on the server,\nwhere it's not needed.\nEnable it on client.
          (bind _returnvalue_3 (Utilities|Time|SetTimerbyEvent (EventDispatchers|CreateEvent :!event Custom|UpdateClient) (Variables|Default|GetTimeBetweenDecalUpdates) true)) ; Start a recurring timer to update visuals
          (Variables|Default|SetUpdateClientTimer _returnvalue_3))))))

(event Custom|UpdateClient
  ;[1 1 0 0.5] [UpdateClient]
  (Variables|Default|GetDecalMID
    (:then
      (bind _fallingobj (Variables|Default|GetFallingObj)
        (:then
          (bind _distance (Class|MissionBlueprintFunctionLibrary|DistanceBetweenTwoVectors (Transformation|GetActorLocation) (Transformation|GetWorldLocation _fallingobj)))
          (Variables|Default|SetPreviousTargetPercentDistanceFromRocket (Variables|Default|GetTargetPercentDistanceFromDropActor))
          (Variables|Default|SetTargetPercentDistanceFromDropActor (Math|Float|MapRangeClamped _distance 0.0 (Variables|Default|GetMaxRangeForMaterialAdjustments) 0.0 1.0))
          (if (!= (Variables|Default|GetTargetPercentDistanceFromDropActor) (Variables|Default|GetPreviousTargetPercentDistanceFromRocket))
            (Rendering|Material|SetScalarParameterbyIndex (Variables|Default|GetDecalMID) (Variables|Default|GetMaterialDistancePercentageParameterIndex) (Variables|Default|GetTargetPercentDistanceFromDropActor))))
        (:else)) ; Can go in and out of relevancy
    )
    (:else)) ; Sanity check.\nShould never become null.
)

(event Custom|DropActorLanded
  ;[0 0 1 0.5] [DropActorLanded] Server and client
  (Actor|SetLifeSpan (Variables|Default|GetPostLandLifespan))
  (Rendering|SetActorHiddenInGame true))

(event EventEndPlay
  (params (EndPlayReason :type EEndPlayReason%20Enum :default "Destroyed"))
  ;[1 0 0 0.5] [EndPlay]
  (Utilities|Time|ClearandInvalidateTimerbyHandle (Variables|Default|GetUpdateClientTimer))
  (bind _dropactor (Variables|Default|GetDropActor)
    (:then
      (Default|UnbindEventfromOnLanded _dropactor (EventDispatchers|CreateEvent :!event Custom|DropActorLanded) :!class /SonicSpearGameplay/Gameplay/Drops/BGA_SonicSpear_MainDropParent.BGA_SonicSpear_MainDropParent_C))
    (:else
      _))
  (bind _physicsdropactor (Variables|Default|GetPhysicsDropActor)
    (:then
      (Default|UnbindEventfromOnLanded _physicsdropactor (EventDispatchers|CreateEvent :!event Custom|DropActorLanded) :!class /SonicSpearGameplay/Gameplay/Drops/B_SonicSpear_Drop_Physics_Parent.B_SonicSpear_Drop_Physics_Parent_C))
    (:else)))

(fn OnRep_DropActor
  (bind _dropactor (Variables|Default|GetDropActor)
    (:then
      (Default|BindEventtoOnLanded _dropactor (EventDispatchers|CreateEvent :!event DropActorLanded) :!class /SonicSpearGameplay/Gameplay/Drops/BGA_SonicSpear_MainDropParent.BGA_SonicSpear_MainDropParent_C))
    (:else)))

(fn OnRep_PhysicsDropActor
  (bind _physicsdropactor (Variables|Default|GetPhysicsDropActor)
    (:then
      (Default|BindEventtoOnLanded _physicsdropactor (EventDispatchers|CreateEvent :!event DropActorLanded) :!class /SonicSpearGameplay/Gameplay/Drops/B_SonicSpear_Drop_Physics_Parent.B_SonicSpear_Drop_Physics_Parent_C))
    (:else)))

(fn ConstructionScript)
