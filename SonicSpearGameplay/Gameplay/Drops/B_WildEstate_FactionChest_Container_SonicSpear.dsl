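; Wild Estate faction chest container for Sonic Spear gameplay, adjusted so it stays openable without colliding with sonic spear actors.
; The construction script calls the parent, then applies the interaction collision profile (noted as FortTriggerOnlyInteractions) to the static mesh.
; This works around BuildingSMActor collision settings overwriting those of BuildingContainer.
; BeginPlay, overlap and tick events are unused, disabled stubs.
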(event EventBeginPlay
  (|CallParentFunction) ; This node is disabled and will not be called.\nDrag off pins to build functionality.
)

(event Collision|EventActorBeginOverlap
  (params (OtherActor :type Actor%20Object%20Reference))
  (|CallParentFunction OtherActor) ; This node is disabled and will not be called.\nDrag off pins to build functionality.
)

(event EventTick
  (params (DeltaSeconds :type Float%20%28single-precision%29))
  (|CallParentFunction DeltaSeconds) ; This node is disabled and will not be called.\nDrag off pins to build functionality.
)

(fn ConstructionScript
  (|CallParentFunction)
  ;[0 0 0 0.5] ABuildingContainer collision is overwritten by collision settings on ABuildingSMActor. We want to disable collision with sonic spear actors, but enable interaction events to open the chest. Set the collision profile to FortTriggerOnlyInteractions on construct.
  (Collision|SetCollisionProfileName (Variables|Debug|GetStaticMeshComponent) (Variables|Default|GetInteractionCollisionProfileName) false :!class /Script/Engine.PrimitiveComponent))
